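# Niagara system for a tentacle slice effect, combining bright slash visuals with a burst of pink-purple gore.
# Slash meshes, cross-shaped sprite sparks and a flash layer over chunks, splats, streaks and GPU smoke thrown from elliptical rings.
# Goo strands partition particles into ribbons, repositioned each update between a local ring and one around a sampled skeletal mesh point.
# System parameters drive slash mesh placement, scale and timing, while emitter-age curves taper the gore spray.
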NiagaraSystem /PistolCanary/Assets/FX/Tentacle_Slice/NS_PC_TentacleSlice
dependencies {
 usedDataInterfaces [
  &NiagaraDataInterfaceCurve
  &NiagaraDataInterfaceColorCurve
  &NiagaraDataInterfaceArrayFloat
  &NiagaraDataInterfaceSpriteRendererInfo
  &NiagaraDataInterfaceSkeletalMesh
 ]
 usedDynamicInputs [&LerpLinearColors, &SelectFloatFromArray]
 usedModules [
  &self:SystemSpawnScript.SystemScriptSource.SystemScriptGraph.NiagaraNodeAssignment_7.SetVariables_#1
  &self:SystemSpawnScript.SystemScriptSource.SystemScriptGraph.NiagaraNodeAssignment_0.SetVariables_#2
  &self:SystemSpawnScript.SystemScriptSource.SystemScriptGraph.NiagaraNodeAssignment_1.SetVariables_#3
  &self:SystemSpawnScript.SystemScriptSource.SystemScriptGraph.NiagaraNodeAssignment_2.SetVariables_#4
  &self:SystemSpawnScript.SystemScriptSource.SystemScriptGraph.NiagaraNodeAssignment_3.SetVariables_#5
  &self:SystemSpawnScript.SystemScriptSource.SystemScriptGraph.NiagaraNodeAssignment_4.SetVariables_#6
  &self:SystemSpawnScript.SystemScriptSource.SystemScriptGraph.NiagaraNodeAssignment_5.SetVariables_#7
  &self:SystemSpawnScript.SystemScriptSource.SystemScriptGraph.NiagaraNodeAssignment_6.SetVariables_#8
  &SystemState
 ]
}
emitters {
 Chunks {
  emitterUpdateScript [
   &EmitterState(Life Cycle Mode=::NewEnumerator1, Loop Behavior=::NewEnumerator1, UseLoopDelay=false)
   &SpawnBurst_Instantaneous(Spawn Count=70, Spawn Time=0.1, Use Spawn Probability=false, UseLoopCountLimit=false)
   &SpawnRate(
    SpawnRate=&FloatFromCurve(
     FloatCurve=DataInterface() {propertyValues {Curve {keys [
         {arriveTangent=-2.4454, interpMode=RCIM_Cubic, leaveTangent=-2.4454, tangentMode=RCTM_User, time=0.15, value=1}
         {arriveTangent=-1.1765, interpMode=RCIM_Cubic, leaveTangent=-1.1765, time=0.5, value=0.35}
         {arriveTangent=-1.1919, tangentMode=RCTM_User, time=1, value=0}
        ]}}}
     CurveIndex=$Emitter.NormalizedLoopAge
     Scale Curve=200
    )
    Use Spawn Probability=false
   )
  ]
  particleSpawnScript [
   /SetVariables(Particles.RandomIntensity:NiagaraFloat=&FloatFromCurve(
     FloatCurve=DataInterface() {propertyValues {Curve {defaultValue=1, keys [<at 0 0 RCIM_Cubic>, {arriveTangent=1.6936, tangentMode=RCTM_User, time=1, value=1}]}}}
     CurveIndex=&RandomRangeFloat(Minimum=0, Maximum=1)
    )) {isSetParameters=true}
   &InitializeParticle(
    Color Maximum=<rgba 2 0.1101 1.7034 1>
    Color Minimum=<rgba 0.25 0.0138 0.169 1>
    Color Mode=::NewEnumerator2
    Lifetime Min=0.2
    Lifetime Mode=::NewEnumerator1
    Sprite Size Mode=::NewEnumerator3
    Uniform Sprite Size=&Multiply_Float(
     A=&Lerp_Float(A=0.5, B=1, Alpha=$Particles.RandomIntensity)
     B=&RandomRangeFloat(Minimum=40, Maximum=170)
    )
    UsePositionOffset=false
    Write Material Random=true
   )
   &ShapeLocation(
    Disc Coverage=0.3
    MaskSpawnGroup=false
    Non Uniform Scale=<1.25 0.9 1>
    Override Seed=false
    Ring Radius=700
    Rotation Axis=<0 0 1>
    Rotation Mode=::NewEnumerator1
    Shape Primitive=::NewEnumerator4
    UseOwnerScale=true
   ) {module=ShapeLocation001}
   &AddVelocity(
    Cone Angle=50
    Cone Axis=<0 1 0.2>
    Module.UseConeFalloff=false
    Module.UseVelDistribution=false
    Override Seed=false
    Velocity Mode=::NewEnumerator2
    Velocity Speed=&Lerp_Float(
     A=2500
     B=25000
     Alpha=&FloatFromCurve(
      FloatCurve=DataInterface() {propertyValues {Curve {keys [
          {arriveTangent=0.5382, interpMode=RCIM_Cubic, leaveTangent=0.5382, tangentMode=RCTM_User, time=0, value=0}
          {arriveTangent=2.6808, interpMode=RCIM_Cubic, leaveTangent=2.6808, tangentMode=RCTM_User, time=1, value=1}
         ]}}}
      CurveIndex=&RandomRangeFloat(Minimum=0, Maximum=1)
      Scale Curve=&Lerp_Float(A=0.6, B=1, Alpha=$Particles.RandomIntensity)
     )
    )
   )
   /SetVariables(Particles.Velocity=&Multiply_VectorByFloat(
     Vector=$Particles.Velocity
     Float=&FloatFromCurve(
      FloatCurve=DataInterface() {propertyValues {Curve {keys [{arriveTangent=-2.7577, interpMode=RCIM_Cubic, leaveTangent=-2.7577, tangentMode=RCTM_User, time=0.15, value=1}, <at 1 0.2>]}}}
      CurveIndex=$Emitter.Age
     )
    )) {isSetParameters=true}
   &InheritVelocity(Inherited Velocity Speed Limit=2000, Speed Limit=true, UseSpeedThreshold=false)
  ]
  particleUpdateScript [
   &ParticleState()
   &GravityForce(Gravity=<0 0 -2200>)
   &Drag(
    Drag=&FloatFromCurve(
     FloatCurve=DataInterface() {propertyValues {Curve {keys [
         {interpMode=RCIM_Cubic, leaveTangent=-8.9049, leaveTangentWeight=0.5018, tangentMode=RCTM_Break, time=0, value=1}
         {interpMode=RCIM_Cubic, tangentMode=RCTM_Break, time=0.3, value=0.05}
        ]}}}
     Scale Curve=7
    )
    Use Linear Drag=true
    Use Rotational Drag=false
   )
   &SolveForcesAndVelocity(Clamp Velocity=false, Limit Acceleration=false)
   &ScaleSpriteSize(Uniform Curve Sprite Scale=DataInterface() {propertyValues {Curve {keys [
        {interpMode=RCIM_Cubic, tangentMode=RCTM_Break, time=0.4, value=1}
        {arriveTangent=-3.9054, interpMode=RCIM_Cubic, tangentMode=RCTM_Break, time=1, value=0}
       ]}}})
   &ScaleColor(
    Scale RGB=&VectorFromFloat(Value=&FloatFromCurve(
      FloatCurve=DataInterface() {propertyValues {Curve {keys [
          {arriveTangent=-1.136, interpMode=RCIM_Cubic, time=0, value=1}
          {arriveTangent=-2, interpMode=RCIM_Cubic, leaveTangent=-2, time=0.2, value=0.3}
          <at 0.5 0>
         ]}}}
      CurveIndex=&Multiply_Float(A=$Particles.Age, B=&RandomRangeFloat(Minimum=0.8, Maximum=1.2))
      Scale Curve=&FloatFromCurve(FloatCurve=DataInterface() {propertyValues {Curve {keys [<at 0 1 RCIM_Cubic>,<at 1 0>]}}}, CurveIndex=$Emitter.Age, Scale Curve=30)
     ))
    ScaleA=false
    ScaleRGB=true
   ) {module=ScaleColor001}
   &DynamicMaterialParameters(
    Index 0 Param 1=&RandomRangeFloat(Minimum=1, Maximum=3)
    Index 0 Param 2=1500
    Index 0 Param 3=10000
    Index 0 Param 4=&FloatFromCurve(FloatCurve=DataInterface() {propertyValues {Curve {keys [<at 0.05 0 RCIM_Cubic>,<at 0.15 1>]}}}, Scale Curve=40)
    Index 1 Param 1=&RandomRangeFloat(Minimum=0, Maximum=1)
    Module.Index0Float0Write=true
    Module.Index0Float1Write=true
    Module.Index0Float2Write=true
    Module.Index0Float3Write=true
    Module.Index1Float0Write=true
    Module.Index1Float1Write=false
    Module.Index1Float2Write=false
    Module.Index1Float3Write=false
    Write Parameter Index 1=::NewEnumerator1
   )
  ]
  rendererClasses [&NiagaraSpriteRendererProperties]
  renderers [{rendererClass=&NiagaraSpriteRendererProperties}]
 }
 Glow {
  bEnabled=false
  emitterUpdateScript [
   &EmitterState(Life Cycle Mode=::NewEnumerator1, Loop Behavior=::NewEnumerator1, Loop Duration=1, UseLoopDelay=false)
   &SpawnBurst_Instantaneous(Use Spawn Probability=false, UseLoopCountLimit=false)
  ]
  particleSpawnScript [
   &InitializeParticle(
    Lifetime=&Multiply_Float(A=0.5, B=$System.FX Time Scale)
    Mesh Scale=&Multiply_VectorByFloat(Vector=<1 1 0.25>, Float=2)
    Mesh Scale Mode=::NewEnumerator4
    Position Offset=$System.FX position
    UsePositionOffset=true
    Write Material Random=true
   )
   &InitialMeshOrientation(Mesh Orientation Mode=::NewEnumerator2, Rotation=$System.CustomRotation, Use Rotation Vector=true)
   &InitialMeshOrientation(Mesh Orientation Mode=::NewEnumerator2, Rotation=<0.5 0 0>, Use Rotation Vector=true) {module=InitialMeshOrientation001}
  ]
  particleUpdateScript [
   &ParticleState()
   &ScaleMeshSize(Scale Factor=&Multiply_Vector(A=$Engine.Owner.Scale, B=$System.SizeScale))
   &ScaleMeshSize(Scale Factor=&VectorFromFloat(Value=$System.FX scale)) {module=ScaleMeshSize001}
   &UpdateMeshOrientation(
    Rotation Rate=&FloatFromCurve(
     FloatCurve=DataInterface() {propertyValues {Curve {keys [{arriveTangent=-3.0754, interpMode=RCIM_Cubic, leaveTangent=-3.0754, tangentMode=RCTM_User, time=0, value=2}, <at 0.5 0.1>]}}}
     Scale Curve=&Divide_Float(A=1.4, B=$System.FX Time Scale)
    )
    Rotation Vector=<0 0 -1>
   ) {module=UpdateMeshOrientation001}
   &ScaleColor(
    Scale Alpha=&FloatFromCurve(FloatCurve=DataInterface() {propertyValues {Curve {keys [
         <at 0 1 RCIM_Cubic>
         {arriveTangent=-4.2795, interpMode=RCIM_Cubic, leaveTangent=-4.2795, tangentMode=RCTM_User, time=0.5844, value=0.5146}
         {arriveTangent=-2.592, interpMode=RCIM_Cubic, leaveTangent=-2.592, tangentMode=RCTM_User, time=1, value=0}
        ]}}})
    Scale RGB=&MakeCustomVectorFromBool(
     Bool=$User.bIsWanderingBoss
     True Vector=&Multiply_VectorByFloat(
      Vector=&MakeVectorFromLinearColorRGB(Color=&ColorFromCurve(ColorCurve=DataInterface() {propertyValues {
          AlphaCurve {keys []}
          BlueCurve {keys [<at 0 0.4 RCIM_Cubic>,<at 0.8853 0.95 RCIM_Cubic>]}
          GreenCurve {keys [<at 0 0.2 RCIM_Cubic>,<at 0.8853 0.6 RCIM_Cubic>]}
          RedCurve {keys [<at 0 1 RCIM_Cubic>,<at 0.8853 2 RCIM_Cubic>]}
         }}))
      Float=&FloatFromCurve(FloatCurve=DataInterface() {propertyValues {Curve {keys [<at 0 1>,<at 0.5 1>]}}})
     )
     False Vector=&Multiply_VectorByFloat(
      Vector=&MakeVectorFromLinearColorRGB(Color=&ColorFromCurve(ColorCurve=DataInterface() {propertyValues {
          AlphaCurve {keys []}
          BlueCurve {keys [<at 0 0.3036 RCIM_Cubic>,<at 0.8853 1.1354 RCIM_Cubic>]}
          GreenCurve {keys [<at 0 0.7679 RCIM_Cubic>,<at 0.8853 1.8559 RCIM_Cubic>]}
          RedCurve {keys [<at 0 1 RCIM_Cubic>,<at 0.8853 2 RCIM_Cubic>]}
         }}))
      Float=&FloatFromCurve(FloatCurve=DataInterface() {propertyValues {Curve {keys [<at 0 1>,<at 0.5 1>]}}}, Scale Curve=3)
     )
    )
    ScaleA=true
    ScaleRGB=true
   )
   &DynamicMaterialParameters(
    Index 0 Param 1=&FloatFromCurve(FloatCurve=DataInterface() {propertyValues {Curve {keys [<at 0 0 RCIM_Cubic>,<at 1 1>]}}})
    Index 0 Param 2=0.1
    Module.Index0Param0Write=true
    Module.Index0Param1Write=true
    Module.Index0Param2Write=true
    Module.Index0Param3Write=true
   )
  ]
  rendererClasses [&NiagaraMeshRendererProperties]
  renderers [{rendererClass=&NiagaraMeshRendererProperties, rendererIndex=0}]
 }
 GooStrands {
  emitterUpdateScript [
   &EmitterState()
   &SpawnBurst_Instantaneous(Spawn Count=30, Use Spawn Probability=false, UseLoopCountLimit=false)
  ]
  particleSpawnScript [
   &PartitionParticles(
    Normalized Index In Partition Includes One=true
    Number Of Partitions=10
    Particles Per Partition=3
    Partition Mode:ENiagara_PartitionMode_v2=::NewEnumerator1
    Write Partition to Ribbon ID=::NewEnumerator0
   )
   /SetVariables(Particles.RandomPerRibbon:NiagaraFloat=&RandomRangeFloat(
     Minimum=0
     Maximum=1
     Randomness Mode=::NewEnumerator1
     Random Seed=&Multiply_Int(A=$Particles.PartitionParticles.Partition, B=31)
     Fixed Random Seed=true
    )) {isSetParameters=true}
   &InitializeParticle(
    Color=<rgba 0 0 0 1>
    Lifetime=&Lerp_Float(A=1, B=0.7, Alpha=$Particles.RandomPerRibbon)
    Ribbon Width=&FloatFromCurve(
     FloatCurve=DataInterface() {propertyValues {Curve {keys [
         <at 0 1 RCIM_Cubic>
         {arriveTangentWeight=0.1667, interpMode=RCIM_Cubic, leaveTangentWeight=0.1667, tangentMode=RCTM_User, time=0.5, value=0.4}
         <at 1 1 RCIM_Cubic>
        ]}}}
     CurveIndex=$Particles.PartitionParticles.Normalized Index In Partition
     Scale Curve=400
    )
    Ribbon Width Mode=::NewEnumerator1
    UsePositionOffset=false
    Write Material Random=true
   )
  ]
  particleUpdateScript [
   &ParticleState()
   &ShapeLocation(
    Apply To Particle Position=::NewEnumerator1
    MaskSpawnGroup=false
    Non Uniform Scale=<1.25 0.9 1>
    Offset=<0 0 -100>
    Offset Coordinate Space=::Local
    Offset Mode=::NewEnumerator0
    Override Seed=false
    Ring / Disc Distribution Mode:ENiagaraSphereDistributionMode=::NewEnumerator1
    Ring Radius=700
    Rotation Axis=<0 0 1>
    Rotation Mode=::NewEnumerator1
    Shape Primitive=::NewEnumerator4
    U Position=$Particles.PartitionParticles.NormalizedPartition
    UseOwnerScale=true
   ) {module=ShapeLocation001}
   &SampleSkeletalMesh(
    AdditionalYawPitchRoll=false
    Bone Sampling Mode:ENiagaraBoneSamplingMode=::NewEnumerator4
    Sampled Position Offset:Vector3f=<-400 200 -800>
    Skeletal Mesh:NiagaraDataInterfaceSkeletalMesh=$System.Skeletal Mesh
   )
   &ShapeLocation(
    Apply To Particle Position=::NewEnumerator1
    MaskSpawnGroup=false
    Non Uniform Scale=<1.25 0.9 1>
    Override Seed=false
    Ring / Disc Distribution Mode:ENiagaraSphereDistributionMode=::NewEnumerator1
    Ring Radius=500
    Rotation Axis=<0 0 1>
    Rotation Mode=::NewEnumerator1
    Shape Origin=$StackContext.SampleSkeletalMesh.SampledPosition
    Shape Primitive=::NewEnumerator4
    U Position=$Particles.PartitionParticles.NormalizedPartition
    UseOwnerScale=true
   )
   /SetVariables(Particles.Position=&Lerp_Vector(
     A=Linked() {linkedVariable {name=Particles.ShapeLocation.ShapePosition, type=NiagaraPosition}}
     B=Linked() {linkedVariable {name=Particles.ShapeLocation001.ShapePosition, type=NiagaraPosition}}
     Alpha=$Particles.Initial.PartitionParticles.Normalized Index In Partition
    )) {isSetParameters=true}
   &DynamicMaterialParameters(
    Index 0 Param 3=&Lerp_Float(A=9, B=15, Alpha=$Particles.RandomPerRibbon)
    Index 0 Param 4=&FloatFromCurve(
     FloatCurve=DataInterface() {propertyValues {Curve {defaultValue=1, keys [<at 0.5 0 RCIM_Cubic>, {tangentMode=RCTM_User, time=1, value=1}]}}}
     Scale Curve=&Add_Float(A=$Particles.RandomPerRibbon, B=1)
    )
    Module.Index0Float0Write=true
    Module.Index0Float1Write=true
    Module.Index0Float2Write=true
    Module.Index0Float3Write=true
   )
   &ScaleRibbonWidth(Ribbon Width Scale=&FloatFromCurve(FloatCurve=DataInterface() {propertyValues {Curve {keys [<at 0 1 RCIM_Cubic>, {arriveTangent=-2.2954, tangentMode=RCTM_User, time=1, value=0}]}}}))
  ]
  rendererClasses [&NiagaraSpriteRendererProperties, &NiagaraRibbonRendererProperties]
  renderers [
   {rendererClass=&NiagaraSpriteRendererProperties}
   {rendererClass=&NiagaraRibbonRendererProperties, rendererIndex=1}
  ]
 }
 HighLight {
  emitterUpdateScript [
   &EmitterState(Life Cycle Mode=::NewEnumerator1, Loop Behavior=::NewEnumerator1, Loop Duration=1, UseLoopDelay=false)
   &SpawnBurst_Instantaneous(Use Spawn Probability=false, UseLoopCountLimit=false)
  ]
  particleSpawnScript [
   &InitializeParticle(
    Color=<rgba 300 40.5911 66.4209 1>
    Lifetime=&Multiply_Float(A=0.2, B=$System.FX Time Scale)
    Mesh Scale=&Multiply_VectorByFloat(Vector=$System.WeaponSize, Float=2.2)
    Mesh Scale Mode=::NewEnumerator4
    Position Offset=$System.FX position
    UsePositionOffset=true
    Write Material Random=true
   )
   &InitialMeshOrientation(Mesh Orientation Mode=::NewEnumerator2, Rotation=$System.CustomRotation, Use Rotation Vector=true)
   &InitialMeshOrientation(Mesh Orientation Mode=::NewEnumerator2, Rotation=<0.5 0 0>, Use Rotation Vector=true) {module=InitialMeshOrientation001}
  ]
  particleUpdateScript [
   &ParticleState()
   &ScaleMeshSize(Scale Factor=&VectorFromFloat(Value=$System.FX scale)) {module=ScaleMeshSize001}
   &ScaleMeshSize(Scale Factor=&Multiply_Vector(A=$Engine.Owner.Scale, B=$System.SizeScale))
   &UpdateMeshOrientation(
    Rotation Rate=&FloatFromCurve(
     FloatCurve=DataInterface() {propertyValues {Curve {keys [{arriveTangent=-3.0754, interpMode=RCIM_Cubic, leaveTangent=-3.0754, tangentMode=RCTM_User, time=0, value=2}, <at 1 0.1>]}}}
     Scale Curve=&Divide_Float(A=1, B=$System.FX Time Scale)
    )
    Rotation Vector=<0 0 -1>
   ) {module=UpdateMeshOrientation001}
   &ScaleColor(
    Scale Alpha=&FloatFromCurve(FloatCurve=DataInterface() {propertyValues {Curve {keys [
         {interpMode=RCIM_Cubic, tangentMode=RCTM_User, time=0, value=1}
         {arriveTangent=-3.6091, interpMode=RCIM_Cubic, leaveTangent=-3.6091, tangentMode=RCTM_User, time=1, value=0}
        ]}}})
    Scale RGB=&MakeVectorFromLinearColorRGB(Color=&ColorFromCurve(ColorCurve=DataInterface() {propertyValues {
        AlphaCurve {keys [<at 0 1 RCIM_Cubic>,<at 1 0 RCIM_Cubic>]}
        BlueCurve {keys [<at 0 1 RCIM_Cubic>,<at 1 1 RCIM_Cubic>]}
        GreenCurve {keys [<at 0 1 RCIM_Cubic>,<at 1 0.0316 RCIM_Cubic>]}
        RedCurve {keys [<at 0 1 RCIM_Cubic>,<at 1 0.0235 RCIM_Cubic>]}
       }}))
    ScaleA=true
    ScaleRGB=true
   )
   &DynamicMaterialParameters(
    Index 0 Param 1=&FloatFromCurve(FloatCurve=DataInterface() {propertyValues {Curve {keys [{tangentMode=RCTM_User, time=0, value=0}, {tangentMode=RCTM_User, time=1, value=1}]}}})
    Index 0 Param 2=0.6
    Module.Index0Param0Write=true
    Module.Index0Param1Write=true
    Module.Index0Param2Write=true
    Module.Index0Param3Write=true
   )
  ]
  rendererClasses [&NiagaraMeshRendererProperties]
  renderers [{rendererClass=&NiagaraMeshRendererProperties, rendererIndex=0}]
 }
 Light001 {
  emitterUpdateScript [
   &EmitterState(Life Cycle Mode=::NewEnumerator1, Loop Behavior=::NewEnumerator1, Loop Duration=1, UseLoopDelay=true)
   &SpawnBurst_Instantaneous(Spawn Count=1, Use Spawn Probability=false)
  ]
  particleSpawnScript [
   &InitializeParticle(
    Color=<rgba 5 5 5 1>
    Lifetime=0.5
    Sprite Size Mode=::NewEnumerator3
    Uniform Sprite Size=30000
    UsePositionOffset=false
    Write Material Random=true
   )
   /SetVariables(Particles.SpriteSize=&Multiply_Vector2D_ByFloat(
     Float=&RemapRange(
      Input Value To Remap=&FN_LOD_Distance_Remap(Start Distance=5000, End Distance=15000, Minimum Output=1, Maximum Output=3)
      Input Min=0
      Input Max=1
      Output Min=0
      Output Max=2
     )
     Vector2D=<5000 5000>
    )) {isSetParameters=true}
   &SystemLocation(MaskSpawnGroup=false)
   /SetVariables(Particles.LightPos:Vector3f=<0 0 1000>) {enabled=false, isSetParameters=true}
  ]
  particleUpdateScript [
   &ParticleState()
   &CameraOffset(Camera Offset Amount=700)
   &ScaleColor(
    Scale Mode=::NewEnumerator1
    Scale RGBA=&Vector4FromCurve(Vector4Curve=DataInterface() {propertyValues {
       AlphaCurve {keys [<at 0.0026 1>, {arriveTangent=-1.9665, interpMode=RCIM_Cubic, leaveTangent=-1.9665, tangentMode=RCTM_User, time=0.0954, value=1}, <at 1 0>]}
       BlueCurve {keys [<at 0.3214 0.0649>,<at 0.6979 5>,<at 1 1>]}
       GreenCurve {keys [<at 0.3214 0>,<at 0.6979 0.7724>,<at 1 0>]}
       RedCurve {keys [<at 0.3214 1.6562>,<at 0.6979 0>,<at 1 0.4596>]}
      }})
    ScaleRGBA=true
   )
   &ScaleSpriteSize(Uniform Curve Sprite Scale=DataInterface() {propertyValues {Curve {keys [
        {arriveTangent=3.5121, interpMode=RCIM_Cubic, leaveTangent=3.5121, tangentMode=RCTM_User, time=0, value=0.3}
        {arriveTangent=-0.0405, interpMode=RCIM_Cubic, leaveTangent=-0.0405, tangentMode=RCTM_User, time=0.1, value=1}
       ]}}})
   &DynamicMaterialParameters(
    Index 0 Param 1=0.5
    Index 0 Param 2=1
    Index 1 Param 1=500
    Module.Param0WriteEnabled=true
    Module.Param0WriteEnabled1=true
    Module.Param1WriteEnabled=true
    Module.Param1WriteEnabled1=true
    Module.Param2WriteEnabled=true
    Module.Param2WriteEnabled1=true
    Module.Param3WriteEnabled=true
    Module.Param3WriteEnabled1=true
    Write Parameter Index 1=true
   )
   /SetVariables(
    Particles.LightColor=&ColorFromCurve(ColorCurve=DataInterface() {propertyValues {
       AlphaCurve {keys [<at 0 1 RCIM_Cubic>,<at 1 0 RCIM_Cubic>]}
       BlueCurve {keys [{arriveTangent=-499, interpMode=RCIM_Cubic, time=0, value=500}, {interpMode=RCIM_Cubic, leaveTangent=-499, time=1, value=1}]}
       GreenCurve {keys [{arriveTangent=-6.2674, interpMode=RCIM_Cubic, time=0, value=359.0213}, {interpMode=RCIM_Cubic, leaveTangent=-6.2674, time=1, value=0.2897}]}
       RedCurve {keys [{arriveTangent=1, interpMode=RCIM_Cubic, time=0, value=2.3431}, {interpMode=RCIM_Cubic, leaveTangent=1, time=1, value=0.6258}]}
      }})
    Particles.LightIntensity_Component:NiagaraFloat=&FloatFromCurve(
     FloatCurve=DataInterface() {propertyValues {Curve {keys [
         {interpMode=RCIM_Cubic, tangentMode=RCTM_User, time=0, value=0}
         {interpMode=RCIM_Cubic, tangentMode=RCTM_User, time=0.1, value=0.7}
         <at 1 0>
        ]}}}
     Scale Curve=100
    )
    Particles.LightRadius=&FloatFromCurve(
     FloatCurve=DataInterface() {propertyValues {Curve {keys [{interpMode=RCIM_Cubic, tangentMode=RCTM_User, time=0, value=0}, {interpMode=RCIM_Cubic, tangentMode=RCTM_User, time=0.1, value=1}]}}}
     Scale Curve=200
    )
   ) {enabled=false, isSetParameters=true}
  ]
  rendererClasses [
   &NiagaraSpriteRendererProperties
   &NiagaraLightRendererProperties
   &NiagaraComponentRendererProperties
  ]
  renderers [
   {rendererClass=&NiagaraSpriteRendererProperties}
   {rendererClass=&NiagaraLightRendererProperties, rendererIndex=1}
   {rendererClass=&NiagaraComponentRendererProperties, rendererIndex=2}
  ]
 }
 Primary_Slice {
  emitterUpdateScript [
   &EmitterState(Life Cycle Mode=::NewEnumerator1, Loop Behavior=::NewEnumerator1, Loop Duration=1, UseLoopDelay=false)
   &SpawnBurst_Instantaneous(Spawn Count=38, Use Spawn Probability=false, UseLoopCountLimit=false)
   &SpawnRate(SpawnRate=20, Use Spawn Probability=false) {enabled=false}
  ]
  particleSpawnScript [
   /Niagara/Modules/Spawn/Initialization/InitializeParticle(
    Color=<rgba 150 0 6.1923 1>
    Lifetime=2
    Module.Write Color=true
    Module.Write Lifetime=true
    Module.Write Mass=true
    Module.Write Material Random=true
    Module.Write Position=true
    Module.Write Scale=false
    Module.Write SpriteRotation=false
    Module.Write SpriteSize=true
    Sprite Size=<300 5>
   ) {enabled=false}
   /Niagara/Modules/Spawn/Initialization/V2/InitializeParticle(
    Color=<rgba 200 0 15.4477 1>
    Lifetime=&RandomRangeFloat(Minimum=0.2, Maximum=0.5)
    Sprite Size=&RandomRangeVector2D(Minimum=<50 10>, Maximum=<200 10>)
    Sprite Size Mode=::NewEnumerator4
    UsePositionOffset=false
    Write Material Random=true
   ) {module=InitializeParticle001}
   &self:CrossShape() {module=CrossShape}
   &AddVelocity(
    Velocity=$Particles.Velocity
    Velocity Speed Scale=&FloatFromCurve(
     FloatCurve=DataInterface() {propertyValues {Curve {keys [{arriveTangent=-2.4578, interpMode=RCIM_Cubic, leaveTangent=-2.4578, tangentMode=RCTM_User, time=0, value=1}, <at 1 0 RCIM_Cubic>]}}}
     CurveIndex=&RandomRangeFloat(Minimum=0, Maximum=1)
     Scale Curve=300000
    )
   )
   /SetVariables(Particles.Velocity=&Add_Vector(
     A=$Particles.Velocity
     B=&RandomVector(Vector Scale=&RandomRangeFloat(Minimum=21.3983, Maximum=27011.5781))
    )) {isSetParameters=true, module=SetVariables_#9}
   &CameraOffset(Camera Offset Amount=1000)
  ]
  particleUpdateScript [
   &ParticleState()
   &ScaleSpriteSize(
    Scale Factor=&MakeVector2D(
     X=&FloatFromCurve(FloatCurve=DataInterface() {propertyValues {Curve {keys [{arriveTangent=-2.1171, interpMode=RCIM_Cubic, leaveTangent=-2.1171, tangentMode=RCTM_User, time=0, value=1}, <at 1 0.25 RCIM_Cubic>]}}})
     Y=&Multiply_Float(A=&VectorLength(Vector=$Particles.Velocity), B=0.01)
    )
    Scale Sprite Size Mode=::NewEnumerator2
   )
   &Drag(Drag=&RandomRangeFloat(Minimum=30, Maximum=10), Use Linear Drag=true, Use Rotational Drag=true)
   &SolveForcesAndVelocity(Clamp Velocity=false, Limit Acceleration=false)
  ]
  rendererClasses [&NiagaraSpriteRendererProperties]
  renderers [{rendererClass=&NiagaraSpriteRendererProperties}]
 }
 Reamin {
  emitterUpdateScript [
   &EmitterState(Life Cycle Mode=::NewEnumerator1, Loop Behavior=::NewEnumerator1, Loop Duration=1, UseLoopDelay=false)
   &SpawnBurst_Instantaneous(Use Spawn Probability=false, UseLoopCountLimit=false)
  ]
  particleSpawnScript [
   &InitializeParticle(
    Color=<rgba 0.016 1.2966 10 1>
    Lifetime=&Multiply_Float(A=0.6, B=$System.FX Time Scale)
    Mesh Scale=&Multiply_VectorByFloat(Vector=<1 1 0.25>, Float=2)
    Mesh Scale Mode=::NewEnumerator4
    Position Offset=$System.FX position
    UsePositionOffset=true
    Write Material Random=true
   )
   /SetVariables(Particles.MeshOrientation=$Engine.Owner.Rotation) {enabled=false, isSetParameters=true}
   &InitialMeshOrientation(Mesh Orientation Mode=::NewEnumerator2, Rotation=$System.CustomRotation, Use Rotation Vector=true)
   &InitialMeshOrientation(Mesh Orientation Mode=::NewEnumerator2, Rotation=<0.5 0 0>, Use Rotation Vector=true) {module=InitialMeshOrientation001}
  ]
  particleUpdateScript [
   &ParticleState()
   &ScaleMeshSize(Scale Factor=&Multiply_Vector(A=$Engine.Owner.Scale, B=$System.SizeScale))
   &ScaleMeshSize(Scale Factor=&VectorFromFloat(Value=$System.FX scale)) {module=ScaleMeshSize001}
   &UpdateMeshOrientation(
    Rotation Rate=&FloatFromCurve(
     FloatCurve=DataInterface() {propertyValues {Curve {keys [{arriveTangent=-3.0754, interpMode=RCIM_Cubic, leaveTangent=-3.0754, tangentMode=RCTM_User, time=0, value=2}, <at 0.5 0.1>]}}}
     Scale Curve=&Divide_Float(A=1.4, B=$System.FX Time Scale)
    )
    Rotation Vector=<0 0 -1>
   ) {module=UpdateMeshOrientation001}
   &ScaleColor(
    Scale Alpha=&FloatFromCurve(
     FloatCurve=DataInterface() {propertyValues {Curve {keys [
         <at 0 1 RCIM_Cubic>
         {arriveTangent=-4.2795, interpMode=RCIM_Cubic, leaveTangent=-4.2795, tangentMode=RCTM_User, time=0.5844, value=0.5146}
         {arriveTangent=-2.592, interpMode=RCIM_Cubic, leaveTangent=-2.592, tangentMode=RCTM_User, time=1, value=0}
        ]}}}
     Scale Curve=3
    )
    ScaleA=true
    ScaleRGB=true
   )
   &DynamicMaterialParameters(
    Index 0 Param 1=&FloatFromCurve(FloatCurve=DataInterface() {propertyValues {Curve {keys [<at 0 0 RCIM_Cubic>,<at 1 1>]}}})
    Index 0 Param 2=0.1
    Module.Index0Param0Write=true
    Module.Index0Param1Write=true
    Module.Index0Param2Write=true
    Module.Index0Param3Write=true
   )
   /SetVariables() {enabled=false, isSetParameters=true}
  ]
  rendererClasses [&NiagaraMeshRendererProperties]
  renderers [{rendererClass=&NiagaraMeshRendererProperties, rendererIndex=0}]
 }
 Secondary_Slice {
  emitterUpdateScript [
   &EmitterState(Life Cycle Mode=::NewEnumerator1, Loop Behavior=::NewEnumerator1, Loop Duration=1, UseLoopDelay=false)
   &SpawnBurst_Instantaneous(Spawn Count=73, Use Spawn Probability=false, UseLoopCountLimit=false)
   &SpawnRate(SpawnRate=20, Use Spawn Probability=false) {enabled=false}
  ]
  particleSpawnScript [
   /Niagara/Modules/Spawn/Initialization/InitializeParticle(
    Color=<rgba 150 0 6.1923 1>
    Lifetime=2
    Module.Write Color=true
    Module.Write Lifetime=true
    Module.Write Mass=true
    Module.Write Material Random=true
    Module.Write Position=true
    Module.Write Scale=false
    Module.Write SpriteRotation=false
    Module.Write SpriteSize=true
    Sprite Size=<300 5>
   ) {enabled=false}
   /Niagara/Modules/Spawn/Initialization/V2/InitializeParticle(
    Color=<rgba 100 0 7.7239 1>
    Lifetime=&RandomRangeFloat(Minimum=0.5, Maximum=2)
    Sprite Size Mode=::NewEnumerator3
    Uniform Sprite Size=&RandomRangeFloat(Minimum=50, Maximum=100)
    UsePositionOffset=false
    Write Material Random=true
   ) {module=InitializeParticle001}
   &self:CrossShape_01() {module=CrossShape_01}
   &self:CrossShape() {enabled=false, module=CrossShape}
   &AddVelocity(
    Velocity=$Particles.Velocity
    Velocity Speed Scale=&FloatFromCurve(
     FloatCurve=DataInterface() {propertyValues {Curve {keys [{arriveTangent=-3.1503, interpMode=RCIM_Cubic, leaveTangent=-3.1503, tangentMode=RCTM_User, time=0, value=1}, <at 1 0 RCIM_Cubic>]}}}
     CurveIndex=&RandomRangeFloat(Minimum=0, Maximum=1)
     Scale Curve=500000
    )
   )
   /SetVariables(Particles.Velocity=&Add_Vector(
     A=$Particles.Velocity
     B=&RandomVector(Vector Scale=&RandomRangeFloat(Minimum=21.3983, Maximum=27011.5781))
    )) {isSetParameters=true, module=SetVariables_#9}
   &CameraOffset(Camera Offset Amount=100)
  ]
  particleUpdateScript [
   &ParticleState()
   &ScaleSpriteSize(
    Uniform Curve Scale=&RandomRangeFloat(Minimum=0, Maximum=1, Evaluation Type=::NewEnumerator1)
    Uniform Curve Sprite Scale=DataInterface() {propertyValues {Curve {keys [{tangentMode=RCTM_User, time=0, value=1}, {tangentMode=RCTM_User, time=1, value=0}]}}}
   ) {module=ScaleSpriteSize001}
   &ScaleSpriteSize(
    Scale Factor=&MakeVector2D(
     X=1
     Y=&Multiply_Float(A=&VectorLength(Vector=$Particles.Velocity), B=0.01)
    )
    Scale Sprite Size Mode=::NewEnumerator2
   ) {enabled=false}
   &Drag(
    Drag=&FloatFromCurve(
     FloatCurve=DataInterface() {propertyValues {Curve {keys [<at 0.1 1>,<at 0.2 0.05>]}}}
     Scale Curve=&RandomRangeFloat(Minimum=30, Maximum=10)
    )
    Use Linear Drag=true
    Use Rotational Drag=true
   )
   &GravityForce(Gravity=<0 0 -2239.1978>)
   &SolveForcesAndVelocity(Clamp Velocity=false, Limit Acceleration=false)
  ]
  rendererClasses [&NiagaraSpriteRendererProperties]
  renderers [{rendererClass=&NiagaraSpriteRendererProperties}]
 }
 Smoke {
  simTarget=GPUComputeSim
  emitterUpdateScript [
   &EmitterState(Life Cycle Mode=::NewEnumerator1, Loop Behavior=::NewEnumerator1, Loop Duration=0.3, UseLoopDelay=false)
   &SpawnBurst_Instantaneous(Spawn Count=20, Spawn Time=0.07, Use Spawn Probability=false, UseLoopCountLimit=false)
   &SpawnRate(SpawnRate=25, Use Spawn Probability=false)
  ]
  particleSpawnScript [
   &InitializeParticle(
    Color Maximum=<rgba 0.9167 0.0621 0.871 1>
    Color Minimum=<rgba 0.3519 0.0375 0.4 1>
    Color Mode=::NewEnumerator2
    Lifetime Max=2.5
    Lifetime Mode=::NewEnumerator1
    Sprite Rotation Mode=::NewEnumerator2
    Sprite Size Mode=::NewEnumerator1
    Uniform Sprite Size Max=3500
    Uniform Sprite Size Min=1500
    UsePositionOffset=false
    Write Material Random=true
   ) {module=InitializeParticle001}
   &InitialMeshOrientation(
    Mesh Orientation Mode=::NewEnumerator2
    Rotation=&RandomRangeVector(Minimum=<-0.02 -0.02 0>, Maximum=<0.02 0.02 1>)
    Use Rotation Vector=true
   )
   &CameraOffset(Camera Offset Amount=250)
   &ShapeLocation(
    MaskSpawnGroup=false
    Non Uniform Scale=<1.25 0.9 1>
    Override Seed=false
    Ring Radius=400
    Rotation Axis=<0 0 1>
    Rotation Mode=::NewEnumerator1
    Shape Primitive=::NewEnumerator4
    UseOwnerScale=true
   ) {module=ShapeLocation001}
   &AddVelocity(
    Module.Use Falloff Distance=false
    Origin Offset=&RandomRangeVector(Minimum=<0 -3000 -500>, Maximum=<0 -3000 300>)
    Override Seed=false
    Velocity Mode=::NewEnumerator1
    Velocity Speed=&RandomRangeFloat(Minimum=5000, Maximum=25000)
   )
   &SubUVAnimation(
    Evaluation Type=::NewEnumerator0
    "Sprite Renderer "=DataInterface() {propertyValues {SpriteRenderer=&self:Smoke_0.NiagaraSpriteRendererProperties_0}}
    SubUV Animation Mode:ENiagara_SubUVLookupModeV2=::NewEnumerator2
    UseEndFrame=false
    UseStartFrame=false
   )
  ]
  particleUpdateScript [
   &ParticleState()
   &ScaleSpriteSize(Uniform Curve Sprite Scale=DataInterface() {propertyValues {Curve {keys [
        {arriveTangent=1.3955, interpMode=RCIM_Cubic, leaveTangent=1.3955, tangentMode=RCTM_User, time=0, value=0.1}
        {arriveTangent=0.3847, tangentMode=RCTM_User, time=1, value=1}
       ]}}})
   &Drag(Drag=4, Use Linear Drag=true, Use Rotational Drag=false)
   &CurlNoiseForce(
    Bias Noise=false
    Bias Noise Field:Vector3f=<0 0 -1>
    Noise Strength=&RandomRangeFloat(Minimum=300, Maximum=600)
    Pan Noise=false
    Pan Noise Field=<1 0 -1>
   )
   &GravityForce(Gravity=&RandomRangeVector(Minimum=<0 0 300>, Maximum=<0 0 -1000>))
   &SolveForcesAndVelocity(Clamp Velocity=false, Limit Acceleration=false)
   &ScaleSpriteSizeBySpeed(
    Max Scale Factor=<0.5 5>
    Sample Scale Factor By Curve=true
    Scale Factor Curve=DataInterface() {propertyValues {Curve {keys [
        {arriveTangent=0.2418, interpMode=RCIM_Cubic, leaveTangent=0.2418, tangentMode=RCTM_User, time=0, value=0}
        {arriveTangent=1.9927, tangentMode=RCTM_User, time=1, value=1}
       ]}}}
    Velocity Threshold=6000
   )
   /SetVariables(Particles.SubImageIndex=&FloatFromCurve(FloatCurve=DataInterface() {propertyValues {Curve {keys [<at 0 0>,<at 1 1>]}}}, Scale Curve=16)) {isSetParameters=true}
   &ScaleColor(
    Scale Alpha=&Multiply_Float(
     A=&FloatFromCurve(
      FloatCurve=DataInterface() {propertyValues {Curve {
         defaultValue=1
         keys [
          {interpMode=RCIM_Cubic, tangentMode=RCTM_User, time=0, value=0}
          {arriveTangent=-2.3653, interpMode=RCIM_Cubic, leaveTangent=-2.3653, tangentMode=RCTM_User, time=0.25, value=1}
          {tangentMode=RCTM_User, time=1, value=0}
         ]
        }}}
      Scale Curve=0.2
     )
     B=1
    )
    Scale RGB=&MakeVectorFromLinearColorRGB(Color=&ColorFromCurve(ColorCurve=DataInterface() {propertyValues {
        AlphaCurve {keys [<at 0 1 RCIM_Cubic>,<at 1 0 RCIM_Cubic>]}
        BlueCurve {keys [<at 0 3 RCIM_Cubic>,<at 0.8568 0.3802 RCIM_Cubic>]}
        GreenCurve {keys [<at 0 0.9708 RCIM_Cubic>,<at 0.8568 0 RCIM_Cubic>]}
        RedCurve {keys [<at 0 0.4769 RCIM_Cubic>,<at 0.8568 0.0523 RCIM_Cubic>]}
       }}))
    ScaleA=true
    ScaleRGB=true
   )
   &DynamicMaterialParameters(
    Index 0 Param 4=&FloatFromCurve(FloatCurve=DataInterface() {propertyValues {Curve {keys [<at 0 1>,<at 1 0>]}}})
    Module.Index0Param0Write=true
    Module.Index0Param1Write=true
    Module.Index0Param2Write=true
    Module.Index0Param3Write=true
   )
  ]
  rendererClasses [&NiagaraSpriteRendererProperties]
  renderers [{rendererClass=&NiagaraSpriteRendererProperties}]
 }
 Splats {
  emitterUpdateScript [
   &EmitterState()
   &SpawnRate(
    SpawnRate=&FloatFromCurve(
     FloatCurve=DataInterface() {propertyValues {Curve {keys [
         <at 0 1 RCIM_Cubic>
         {arriveTangent=-0.1215, interpMode=RCIM_Cubic, leaveTangent=-0.1215, tangentMode=RCTM_User, time=0.75, value=0.6}
         {arriveTangent=-0.6834, interpMode=RCIM_Cubic, leaveTangent=-0.6834, tangentMode=RCTM_User, time=2.5, value=0}
        ]}}}
     CurveIndex=$Emitter.Age
     Scale Curve=20
    )
    Use Spawn Probability=false
   )
   &SpawnBurst_Instantaneous(Spawn Count=8, Spawn Time=0.12, Use Spawn Probability=false, UseLoopCountLimit=false)
  ]
  particleSpawnScript [
   &InitializeParticle(
    Color=<rgba 0.25 0 0.0625 1>
    Lifetime Max=1
    Lifetime Min=0.7
    Lifetime Mode=::NewEnumerator1
    Sprite Rotation Mode=::NewEnumerator2
    Sprite Size Max=<2000 2000>
    Sprite Size Min=<400 400>
    Sprite Size Mode=::NewEnumerator2
    Sprite UV Mode=::NewEnumerator3
    UsePositionOffset=false
    Write Material Random=true
   )
   &ShapeLocation(
    MaskSpawnGroup=false
    Non Uniform Scale=<1.25 0.9 1>
    Override Seed=false
    Ring Radius=400
    Rotation Axis=<0 0 1>
    Rotation Mode=::NewEnumerator1
    Shape Primitive=::NewEnumerator4
    UseOwnerScale=true
   ) {module=ShapeLocation001}
   &AddVelocity(
    Module.Use Falloff Distance=false
    Origin Offset=&RandomRangeVector(Minimum=<0 -3000 -2500>, Maximum=<0 -3000 -1500>)
    Override Seed=false
    Velocity Mode=::NewEnumerator1
    Velocity Speed=&FloatFromCurve(
     FloatCurve=DataInterface() {propertyValues {Curve {keys [<at 0.15 1 RCIM_Cubic>,<at 0.5 0.2>]}}}
     CurveIndex=$Emitter.Age
     Scale Curve=&RandomRangeFloat(Minimum=1000, Maximum=6000)
    )
   )
  ]
  particleUpdateScript [
   &ParticleState()
   &SpriteRotationRate(Rotation Rate=&FloatFromCurve(
     FloatCurve=DataInterface() {propertyValues {Curve {keys [{arriveTangent=-1.4875, interpMode=RCIM_Cubic, leaveTangent=-1.4875, tangentMode=RCTM_User, time=0, value=1}, <at 1 0.2>]}}}
     Scale Curve=&RandomRangeFloat(Minimum=-250, Maximum=250)
    ))
   &ScaleSpriteSize(Uniform Curve Sprite Scale=DataInterface() {propertyValues {Curve {keys [{arriveTangent=1.5198, time=0, value=0.1}, {arriveTangent=0.5626, tangentMode=RCTM_User, time=1, value=1}]}}})
   &Drag(Drag=2, Use Linear Drag=true, Use Rotational Drag=false)
   &GravityForce(Gravity=<0 0 -2500>)
   &SolveForcesAndVelocity(Clamp Velocity=false, Limit Acceleration=false)
   &DynamicMaterialParameters(
    Index 0 Param 1=&RandomRangeFloat(Minimum=0, Maximum=1)
    Index 0 Param 3=&FloatFromCurve(
     FloatCurve=DataInterface() {propertyValues {Curve {defaultValue=1, keys [<at 0 0 RCIM_Cubic>, {arriveTangent=0.7412, tangentMode=RCTM_User, time=1, value=1}]}}}
     Scale Curve=3
    )
    Index 0 Param 4=&FloatFromCurve(FloatCurve=DataInterface() {propertyValues {Curve {defaultValue=1, keys [<at 0 0 RCIM_Cubic>, {arriveTangent=1.4949, tangentMode=RCTM_User, time=1, value=1}]}}})
    Module.Index0Float0Write=true
    Module.Index0Float1Write=false
    Module.Index0Float2Write=true
    Module.Index0Float3Write=true
   )
   &CameraOffset(Camera Offset Amount=&Multiply_Float(
     A=/Niagara/DynamicInputs/Math/CalculateParticleRadius(
      Radius Calculation Type:ENiagaraCollisionRadiusOptions=::NewEnumerator0
      Particle Radius:NiagaraFloat=1
      Mesh Dimensions:Vector3f=<1 1 1>
      Method for Calculating Particle Radius:ENiagaraCalculateRadiusOptions=::NewEnumerator2
     )
     B=0.5
    ))
  ]
  rendererClasses [&NiagaraSpriteRendererProperties]
  renderers [{rendererClass=&NiagaraSpriteRendererProperties}]
 }
 Spline_Streaks {
  emitterUpdateScript [
   &EmitterState(Life Cycle Mode=::NewEnumerator1, Loop Behavior=::NewEnumerator1, Loop Duration=0.7, UseLoopDelay=false)
   &SpawnBurst_Instantaneous(Spawn Count=16, Spawn Time=0.1, Use Spawn Probability=false, UseLoopCountLimit=false)
   &SpawnRate(SpawnRate=20, Use Spawn Probability=false)
  ]
  particleSpawnScript [
   /SetVariables(Particles.RandomIntensity:NiagaraFloat=&FloatFromCurve(
     FloatCurve=DataInterface() {propertyValues {Curve {defaultValue=1, keys [<at 0 0 RCIM_Cubic>, {arriveTangent=1.6936, tangentMode=RCTM_User, time=1, value=1}]}}}
     CurveIndex=&RandomRangeFloat(Minimum=0, Maximum=1)
    )) {isSetParameters=true}
   &InitializeParticle(
    Color=<rgba 1.5371 0.0687 2 1>
    Lifetime=&Multiply_Float(
     A=&RandomRangeFloat(Minimum=0.4, Maximum=0.5)
     B=&Lerp_Float(A=0.6, B=1, Alpha=$Particles.RandomIntensity)
    )
    Mesh Scale Mode=::NewEnumerator3
    Mesh Uniform Scale=&Multiply_Float(
     A=&FloatFromCurve(
      FloatCurve=DataInterface() {propertyValues {Curve {keys [<at 0 2 RCIM_Cubic>,<at 0.15 1>]}}}
      CurveIndex=$Emitter.Age
      Scale Curve=&RandomRangeFloat(Minimum=0.7, Maximum=1)
     )
     B=&Lerp_Float(A=15, B=30, Alpha=$Particles.RandomIntensity)
    )
    UsePositionOffset=false
    Write Material Random=true
   )
   &ShapeLocation(
    MaskSpawnGroup=false
    Non Uniform Scale=<1.25 0.9 1>
    Override Seed=false
    Ring Radius=700
    Rotation Axis=<0 0 1>
    Rotation Mode=::NewEnumerator1
    Shape Primitive=::NewEnumerator4
    UseOwnerScale=true
   ) {module=ShapeLocation001}
   &AddVelocity(
    Module.Use Falloff Distance=false
    Origin Offset=&RandomRangeVector(Minimum=<0 -1000 -200>, Maximum=<0 -1000 50>)
    Override Seed=false
    Velocity Mode=::NewEnumerator1
    Velocity Speed=&Lerp_Float(A=1200, B=5000, Alpha=$Particles.RandomIntensity)
   )
   /SetVariables(Particles.Velocity=&Multiply_VectorByFloat(
     Vector=$Particles.Velocity
     Float=&FloatFromCurve(
      FloatCurve=DataInterface() {propertyValues {Curve {keys [{arriveTangent=-2.5713, interpMode=RCIM_Cubic, leaveTangent=-2.5713, tangentMode=RCTM_User, time=0.1, value=1}, <at 0.4 0.2>]}}}
      CurveIndex=$Emitter.Age
     )
    )) {isSetParameters=true}
  ]
  particleUpdateScript [
   &ParticleState()
   &UpdateMeshOrientation(
    Rotation Rate=&FloatFromCurve(
     FloatCurve=DataInterface() {propertyValues {Curve {keys [
         {arriveTangent=-2.1925, interpMode=RCIM_Cubic, leaveTangent=-2.1925, tangentMode=RCTM_User, time=0, value=1}
         {arriveTangent=-0.2845, interpMode=RCIM_Cubic, leaveTangent=-0.2845, tangentMode=RCTM_User, time=1, value=0.05}
        ]}}}
     Scale Curve=&RandomRangeFloat(Minimum=0.15, Maximum=0.25)
    )
    Rotation Vector=<0 1 -1>
   )
   &ScaleMeshSize(Scale Factor=&VectorFromFloat(Value=&FloatFromCurve(FloatCurve=DataInterface() {propertyValues {Curve {
         defaultValue=1
         keys [
          {arriveTangent=1.8176, interpMode=RCIM_Cubic, leaveTangent=1.8176, tangentMode=RCTM_User, time=0, value=0}
          {arriveTangent=0.506, tangentMode=RCTM_User, time=1, value=1}
         ]
        }}})))
   &GravityForce(Gravity=&RandomRangeVector(Minimum=<0 0 -400>, Maximum=<0 0 -300>))
   &Drag(Drag=&RandomRangeFloat(Minimum=0.25, Maximum=0.5), Use Linear Drag=true, Use Rotational Drag=false)
   &SolveForcesAndVelocity(Clamp Velocity=false, Limit Acceleration=false)
   &DynamicMaterialParameters(
    Index 0 Param 1=&FloatFromCurve(FloatCurve=DataInterface() {propertyValues {Curve {keys [<at 0.4 0 RCIM_Cubic>, {arriveTangent=2.3043, interpMode=RCIM_Cubic, leaveTangent=2.3043, tangentMode=RCTM_User, time=1, value=1}]}}})
    Index 0 Param 2=&FloatFromCurve(
     FloatCurve=DataInterface() {propertyValues {Curve {keys [
         {arriveTangent=1.3993, interpMode=RCIM_Cubic, leaveTangent=1.3993, tangentMode=RCTM_User, time=0, value=0.5}
         {interpMode=RCIM_Cubic, tangentMode=RCTM_User, time=1, value=1}
        ]}}}
     Scale Curve=&RandomRangeFloat(Minimum=0, Maximum=2)
    )
    Index 0 Param 3=&Multiply_Float(
     A=&MakeFloatFromVector(Vector=$Particles.Initial.Scale, Channel=::NewEnumerator0)
     B=&RandomRangeFloat(Minimum=0.07, Maximum=0.22)
    )
    Index 0 Param 4=&FloatFromCurve(
     FloatCurve=DataInterface() {propertyValues {Curve {keys [
         {interpMode=RCIM_Cubic, leaveTangentWeight=0.7017, tangentMode=RCTM_User, time=0, value=0}
         {arriveTangent=1.7252, arriveTangentWeight=1.2352, interpMode=RCIM_Cubic, leaveTangent=1.7252, tangentMode=RCTM_User, time=1, value=1}
        ]}}}
     Scale Curve=&RandomRangeFloat(Minimum=0.75, Maximum=2)
    )
    Index 1 Param 1=&FloatFromCurve(
     FloatCurve=DataInterface() {propertyValues {Curve {keys [
         {arriveTangent=2.4162, interpMode=RCIM_Cubic, leaveTangent=2.4162, tangentMode=RCTM_User, time=0, value=0}
         {interpMode=RCIM_Cubic, tangentMode=RCTM_User, time=1, value=1}
        ]}}}
     Scale Curve=0
    )
    Index 1 Param 2=&RandomRangeFloat(Minimum=4, Maximum=8)
    Module.Index0Float0Write=true
    Module.Index0Float1Write=true
    Module.Index0Float2Write=true
    Module.Index0Float3Write=true
    Module.Index1Float0Write=true
    Module.Index1Float1Write=true
    Module.Index1Float2Write=false
    Module.Index1Float3Write=false
    Write Parameter Index 1=::NewEnumerator1
   )
  ]
  rendererClasses [&NiagaraMeshRendererProperties]
  renderers [{rendererClass=&NiagaraMeshRendererProperties, rendererIndex=0}]
 }
}
summary {
 systemName=NS_PC_TentacleSlice
 emitters [
  {emitterName=Reamin, rendererClasses [&NiagaraMeshRendererProperties]}
  {emitterName=HighLight, rendererClasses [&NiagaraMeshRendererProperties]}
  {bEnabled=false, emitterName=Glow, rendererClasses [&NiagaraMeshRendererProperties]}
  {
   emitterName=Light001
   rendererClasses [
    &NiagaraSpriteRendererProperties
    &NiagaraLightRendererProperties
    &NiagaraComponentRendererProperties
   ]
  }
  {emitterName=Primary_Slice, rendererClasses [&NiagaraSpriteRendererProperties]}
  {emitterName=Secondary_Slice, rendererClasses [&NiagaraSpriteRendererProperties]}
  {emitterName=Spline_Streaks, rendererClasses [&NiagaraMeshRendererProperties]}
  {emitterName=Chunks, rendererClasses [&NiagaraSpriteRendererProperties]}
  {emitterName=GooStrands, rendererClasses [&NiagaraSpriteRendererProperties, &NiagaraRibbonRendererProperties]}
  {emitterName=Smoke, simTarget=GPUComputeSim, rendererClasses [&NiagaraSpriteRendererProperties]}
  {emitterName=Splats, rendererClasses [&NiagaraSpriteRendererProperties]}
 ]
 userVariables [
  User.bIsWanderingBoss {type=NiagaraBool}
  User.ScaleMult {type=NiagaraFloat, defaultValue=1, description=""}
 ]
}
